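# Physics asset for the left axe mesh of the FNCS Season 32 pickaxe, a Fortnite melee weapon cosmetic.
# It defines a single capsule body on the Melee_TwoHanded_Handle bone, offset and rotated from the bone origin.
# With only one body and no constraints, it gives the axe one simple collision volume rather than articulated physics.
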PhysicsAsset /BRCosmetics/Weapons/FORT_Melee/Pickaxe_FNCS_S32/Meshes/Seperated/FNCS_S32_L_Axe_PhysicsAsset
bodies {Melee_TwoHanded_Handle [{shapeType=Capsule, length=55.9287, radius=13.9943, center=<8.4846 -7.8529 55.2558>, rotation=<rot -0.3567 -90 2.3647>}]}
constraints []
