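# Physics asset for the right-hand axe mesh of the FNCS S32 pickaxe cosmetic in Fortnite melee weapons.
# It defines a single capsule body on the Melee_TwoHanded_Handle bone, offset and rotated to enclose the axe.
# With no constraints, the weapon acts as one rigid collision shape.
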PhysicsAsset /BRCosmetics/Weapons/FORT_Melee/Pickaxe_FNCS_S32/Meshes/Seperated/FNCS_S32_R_Axe_PhysicsAsset
bodies {Melee_TwoHanded_Handle [{shapeType=Capsule, length=62.1131, radius=12.5421, center=<-7.6961 -8.4035 61.3286>, rotation=<rot 0.3591 -90 -3.5921>}]}
constraints []
